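# Sound cue for the Force Push blast, layering close, medium and distant variations of the sound.
# An IsLocal branch picks quieter 2D-attenuated blast waves for the local player, or pitch-varied close waves otherwise.
# That branch is mixed with medium and low-level distant layers, each layer using random waves, its own attenuation and a weapon-fire sound class override.
# A concurrency override limits simultaneous plays, stopping the oldest.
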SoundCue /PlotTwist/ForcePowers/ForcePush/Sound/ForcePush_Blast_Cue
root=Mixer_2
nodes [
 {
  id=Attenuation_0
  children [Modulator_13]
  properties {attenuationSettings=/Game/Sounds/Attenuation/Fort_ForcePowers_ForcePush_Blast_Medium_Att}
 }
 {
  id=Attenuation_1
  children [Modulator_1]
  properties {attenuationSettings=/Game/Sounds/Attenuation/Fort_ForcePowers_ForcePush_Blast_Close_Athena_Att}
 }
 {id=Attenuation_2, children [Modulator_0], properties {attenuationSettings=/Game/Sounds/Attenuation/Fort_ForcePowers_ForcePush_Cast_2D_Att}}
 {
  id=Attenuation_3
  children [Modulator_2]
  properties {attenuationSettings=/Game/Sounds/Attenuation/Fort_ForcePowers_ForcePush_Blast_Distant_Att}
 }
 {id=Branch_0, children [Attenuation_2, Attenuation_1, Attenuation_1], properties {boolParameterName=IsLocal}}
 {id=Mixer_2, children [SoundClass_77, SoundClass_76, SoundClass_0], properties {inputVolume [1.5,1,0.05]}}
 {id=Modulator_0, children [Random_0], properties {pitchMax=1, pitchMin=1, volumeMax=0.5, volumeMin=0.5}}
 {id=Modulator_1, children [Random_1], properties {pitchMax=1.1, pitchMin=0.9, volumeMax=1, volumeMin=1}}
 {id=Modulator_13, children [Random_3], properties {volumeMax=1, volumeMin=1}}
 {id=Modulator_2, children [Random_2], properties {volumeMax=1, volumeMin=1}}
 {id=Random_0, children [WavePlayer_6,WavePlayer_7,WavePlayer_8], properties {weights [1,1,1]}}
 {id=Random_1, children [WavePlayer_12, WavePlayer_13, WavePlayer_14], properties {weights [1,1,1]}}
 {id=Random_2, children [WavePlayer_18, WavePlayer_19, WavePlayer_20], properties {weights [1,1,1]}}
 {id=Random_3, children [WavePlayer_15, WavePlayer_16, WavePlayer_17], properties {weights [1,1,1]}}
 {
  id=SoundClass_0
  children [Attenuation_3]
  properties {soundClassOverride=&Fort_WeaponRanged_Fire_Distant}
 }
 {
  id=SoundClass_76
  children [Attenuation_0]
  properties {soundClassOverride=&Fort_WeaponRanged_Fire_Distant}
 }
 {id=SoundClass_77, children [Branch_0], properties {soundClassOverride=&Fort_WeaponRanged_Fire_3P_PSM}}
 {id=WavePlayer_12, properties {soundWaveAssetPtr=&/ForcePush_Blast_Close_01}}
 {id=WavePlayer_13, properties {soundWaveAssetPtr=&/ForcePush_Blast_Close_02}}
 {id=WavePlayer_14, properties {soundWaveAssetPtr=&/ForcePush_Blast_Close_03}}
 {id=WavePlayer_15, properties {soundWaveAssetPtr=&/ForcePush_Blast_Medium_01}}
 {id=WavePlayer_16, properties {soundWaveAssetPtr=&/ForcePush_Blast_Medium_02}}
 {id=WavePlayer_17, properties {soundWaveAssetPtr=&/ForcePush_Blast_Medium_03}}
 {id=WavePlayer_18, properties {soundWaveAssetPtr=&/ForcePush_Blast_Distant_01}}
 {id=WavePlayer_19, properties {soundWaveAssetPtr=&/ForcePush_Blast_Distant_02}}
 {id=WavePlayer_20, properties {soundWaveAssetPtr=&/ForcePush_Blast_Distant_03}}
 {id=WavePlayer_6, properties {soundWaveAssetPtr=&/ForcePush_Blast_03}}
 {id=WavePlayer_7, properties {soundWaveAssetPtr=&/ForcePush_Blast_01}}
 {id=WavePlayer_8, properties {soundWaveAssetPtr=&/ForcePush_Blast_02}}
]
properties {
 bOverrideConcurrency=true
 maxDistance=2097152
 soundClassObject=&Fort_Gadgets_Items
 volumeMultiplier=1.1
 attenuationOverrides {
  bAttenuate=false
  distanceAlgorithm=NaturalSound
  falloffDistance=3800
  lPFRadiusMax=3000
  lPFRadiusMin=2000
  attenuationShapeExtents {x=2200}
 }
 concurrencyOverrides {maxCount=2, resolutionRule=StopOldest}
}
