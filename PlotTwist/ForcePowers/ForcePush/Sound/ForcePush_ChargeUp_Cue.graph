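# ForcePush_ChargeUp_Cue is the Sound Cue for the charge-up audio of the Force Push force power.
# A Random node picks one of three equally weighted charge-up waves, and a Modulator above it varies pitch for each play.
# The cue uses the Force Push close blast attenuation asset and a third-person ranged weapon fire sound class, with overridden concurrency capping simultaneous instances.
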SoundCue /PlotTwist/ForcePowers/ForcePush/Sound/ForcePush_ChargeUp_Cue
root=Modulator_1
nodes [
 {id=Modulator_1, children [Random_0], properties {pitchMax=1.15}}
 {id=Random_0, children [WavePlayer_0,WavePlayer_2,WavePlayer_3], properties {weights [1,1,1]}}
 {id=WavePlayer_0, properties {soundWaveAssetPtr=&/ForcePush_ChargeUp_01}}
 {id=WavePlayer_2, properties {soundWaveAssetPtr=&/ForcePush_ChargeUp_02}}
 {id=WavePlayer_3, properties {soundWaveAssetPtr=&/ForcePush_ChargeUp_03}}
]
properties {
 attenuationSettings=/Game/Sounds/Attenuation/Fort_ForcePowers_ForcePush_Blast_Close_Athena_Att
 bOverrideConcurrency=true
 maxDistance=4125
 pitchMultiplier=1.02
 soundClassObject=&Fort_WeaponRanged_Fire_3P_PSM
 volumeMultiplier=1.2
 attenuationOverrides {falloffDistance=12000, attenuationShapeExtents {x=500}}
 concurrencyOverrides {maxCount=3}
}
