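# Sound cue for a Force Push impact on a player, built from layered impact and wind sounds.
# Two randomized impact layers with different attenuation and gain mix with a slightly delayed wind layer, which has its own attenuation.
# A quality-level node switches the wind between five-variation and three-variation random pools.
# Each branch overrides its sound class to a third-person ranged-weapon fire class, and the cue caps its own concurrency.
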SoundCue /PlotTwist/ForcePowers/ForcePush/Sound/ForcePush_Impact_Player_Cue
root=Mixer_1
nodes [
 {id=Attenuation_0, children [Modulator_0], properties {attenuationSettings=/Game/Sounds/Attenuation/Fort_ForcePowers_ForcePush_Blast_3P_Att}}
 {
  id=Attenuation_1
  children [SoundClass_2]
  properties {attenuationSettings=/Game/Sounds/Attenuation/Fort_ForcePowers_ForcePush_Blast_Impact_Player_Att}
 }
 {
  id=Attenuation_2
  children [SoundClass_1]
  properties {attenuationSettings=/Game/Sounds/Attenuation/Fort_ForcePowers_ForcePush_Blast_Impact_PlayerWind_Att}
 }
 {id=Delay_2, children [QualityLevel_0], properties {delayMax=0.15, delayMin=0.15}}
 {id=Mixer_0, children [Attenuation_1, Attenuation_2], properties {inputVolume [4,1]}}
 {id=Mixer_1, children [SoundClass_0,Mixer_0], properties {inputVolume [4,3.4]}}
 {id=Modulator_0, children [Random_1], properties {pitchMax=1, volumeMax=0.4, volumeMin=0.4}}
 {id=Modulator_1, children [Random_2], properties {pitchMax=1, volumeMax=2, volumeMin=2}}
 {id=Modulator_2, children [Delay_2], properties {pitchMax=1, volumeMax=1, volumeMin=1}}
 {id=QualityLevel_0, children [Random_3,Random_3,Random_4,Random_4]}
 {id=Random_1, children [WavePlayer_0,WavePlayer_1,WavePlayer_2], properties {weights [1,1,1]}}
 {id=Random_2, children [WavePlayer_3,WavePlayer_4,WavePlayer_5], properties {weights [1,1,1]}}
 {id=Random_3, children [WavePlayer_7, WavePlayer_8, WavePlayer_9, WavePlayer_10, WavePlayer_11], properties {weights [1,1,1,1,1]}}
 {id=Random_4, children [WavePlayer_9, WavePlayer_10, WavePlayer_11], properties {bShouldExcludeFromBranchCulling=true, weights [1,1,1]}}
 {
  id=SoundClass_0
  children [Attenuation_0]
  properties {soundClassOverride=&Fort_WeaponRanged_Fire_3P_PSM}
 }
 {id=SoundClass_1, children [Modulator_2], properties {soundClassOverride=&Fort_WeaponRanged_Fire_3P_PSM}}
 {id=SoundClass_2, children [Modulator_1], properties {soundClassOverride=&Fort_WeaponRanged_Fire_3P_PSM}}
 {id=WavePlayer_0, properties {soundWaveAssetPtr=&/ForcePush_ImpactPlayer_01}}
 {id=WavePlayer_1, properties {soundWaveAssetPtr=&/ForcePush_ImpactPlayer_02}}
 {id=WavePlayer_10, properties {soundWaveAssetPtr=&/ForcePush_ImpactPlayerWiind_04}}
 {id=WavePlayer_11, properties {soundWaveAssetPtr=&/ForcePush_ImpactPlayerWiind_05}}
 {id=WavePlayer_2, properties {soundWaveAssetPtr=&/ForcePush_ImpactPlayer_03}}
 {id=WavePlayer_3, properties {soundWaveAssetPtr=&ForcePush_ImpactPlayer_01}}
 {id=WavePlayer_4, properties {soundWaveAssetPtr=&ForcePush_ImpactPlayer_02}}
 {id=WavePlayer_5, properties {soundWaveAssetPtr=&ForcePush_ImpactPlayer_03}}
 {id=WavePlayer_7, properties {soundWaveAssetPtr=&/ForcePush_ImpactPlayerWiind_01}}
 {id=WavePlayer_8, properties {soundWaveAssetPtr=&/ForcePush_ImpactPlayerWiind_02}}
 {id=WavePlayer_9, properties {soundWaveAssetPtr=&/ForcePush_ImpactPlayerWiind_03}}
]
properties {
 bOverrideConcurrency=true
 maxDistance=6500
 pitchMultiplier=0.9
 soundClassObject=&Fort_Impacts
 volumeMultiplier=1
 attenuationOverrides {falloffDistance=6000, attenuationShapeExtents {x=1000}}
 concurrencyOverrides {maxCount=3}
}
