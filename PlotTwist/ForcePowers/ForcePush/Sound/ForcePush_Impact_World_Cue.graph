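# Sound cue for the Force Push world impact, playing one of five equally weighted impact variations.
# The random pick passes through a fixed short delay, a modulator that boosts volume and varies pitch slightly, a sound class node and an attenuation node.
# That sound class node specifies a third-person ranged-weapon fire class, while the cue itself uses the impacts class.
# A concurrency override caps simultaneous instances at three.
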SoundCue /PlotTwist/ForcePowers/ForcePush/Sound/ForcePush_Impact_World_Cue
root=Attenuation_0
nodes [
 {
  id=Attenuation_0
  children [SoundClass_0]
  properties {attenuationSettings=/Game/Sounds/Attenuation/Fort_ForcePowers_ForcePush_Blast_Impact_World_Att}
 }
 {id=Delay_0, children [Random_1], properties {delayMax=0.4, delayMin=0.4}}
 {id=Modulator_0, children [Delay_0], properties {pitchMax=1.1, pitchMin=0.9, volumeMax=2, volumeMin=2}}
 {id=Random_1, children [WavePlayer_0,WavePlayer_1,WavePlayer_2,WavePlayer_3,WavePlayer_4], properties {weights [1,1,1,1,1]}}
 {id=SoundClass_0, children [Modulator_0], properties {soundClassOverride=&Fort_WeaponRanged_Fire_3P_PSM}}
 {id=WavePlayer_0, properties {soundWaveAssetPtr=&/ForcePush_ImpactWorld_01}}
 {id=WavePlayer_1, properties {soundWaveAssetPtr=&/ForcePush_ImpactWorld_02}}
 {id=WavePlayer_2, properties {soundWaveAssetPtr=&/ForcePush_ImpactWorld_03}}
 {id=WavePlayer_3, properties {soundWaveAssetPtr=&/ForcePush_ImpactWorld_04}}
 {id=WavePlayer_4, properties {soundWaveAssetPtr=&/ForcePush_ImpactWorld_05}}
]
properties {
 bOverrideConcurrency=true
 maxDistance=4100
 soundClassObject=&Fort_Impacts
 volumeMultiplier=1
 attenuationOverrides {falloffDistance=6000, attenuationShapeExtents {x=1000}}
 concurrencyOverrides {maxCount=3}
}
