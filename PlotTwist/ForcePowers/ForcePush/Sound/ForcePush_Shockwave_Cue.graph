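# Sound cue for the Force Push shockwave in the PlotTwist Force Powers.
# A Random node picks among four equally weighted shockwave waves and feeds a Modulator that slightly varies pitch while holding volume steady.
# A Sound Class root node routes it to the third-person ranged weapon fire class.
# The cue boosts volume, references a dedicated attenuation asset and overrides concurrency to cap simultaneous instances.
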SoundCue /PlotTwist/ForcePowers/ForcePush/Sound/ForcePush_Shockwave_Cue
root=SoundClass_0
nodes [
 {id=Modulator_0, children [Random_0], properties {pitchMax=1.1, pitchMin=0.9, volumeMax=1, volumeMin=1}}
 {id=Random_0, children [WavePlayer_0,WavePlayer_2,WavePlayer_6,WavePlayer_7], properties {weights [1,1,1,1]}}
 {id=SoundClass_0, children [Modulator_0], properties {soundClassOverride=&Fort_WeaponRanged_Fire_3P_PSM}}
 {id=WavePlayer_0, properties {soundWaveAssetPtr=&/ForcePush_ShockWave_01}}
 {id=WavePlayer_2, properties {soundWaveAssetPtr=&/ForcePush_ShockWave_02}}
 {id=WavePlayer_6, properties {soundWaveAssetPtr=&/ForcePush_ShockWave_03}}
 {id=WavePlayer_7, properties {soundWaveAssetPtr=&/ForcePush_ShockWave_04}}
]
properties {
 attenuationSettings=/Game/Sounds/Attenuation/Fort_ForcePowers_ForcePush_Shockwave_Att
 bOverrideConcurrency=true
 maxDistance=5500
 soundClassObject=&Fort_WeaponRanged_Fire_3P_PSM
 volumeMultiplier=2
 attenuationOverrides {falloffDistance=6000, attenuationShapeExtents {x=1000}}
 concurrencyOverrides {maxCount=3}
}
